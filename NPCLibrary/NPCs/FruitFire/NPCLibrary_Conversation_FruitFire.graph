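# Conversation database for the FruitFire NPC, defining its greeting and the dialogue branches it offers.
# The entry point leads to one speech node that varies its line by first-ever meeting, match conversation count or the player's outfit, with curve-driven priorities.
# The speech links to remote entries for NPC library quests, basic services, Season 38 FruitFire P04 and L&F collection, the last hidden while an L&F quest is active.
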ConversationDatabase /NPCLibrary/NPCs/FruitFire/NPCLibrary_Conversation_FruitFire
entry_points [{tag=Conversation.Entry.FruitFire, destinations [#1]}]
nodes [
 ConversationEntryPointNode_7 {
  id=#1
  class=ConversationEntryPointNode
  outputs [#2]
  properties {entryTag {tagName=Conversation.Entry.FruitFire}}
 }
 Speech_C_4 {
  id=#2
  class=Speech_C
  outputs [#3, #4, #5, #6]
  properties {
   nodeName=""
   generalConfig {
    defaultMessage=Which member of my family do I need to apologize for now?
    contextualMessages [
     {
      message=This town is a part of us all...
      contextRequirements [{requirement {class=FortControllerRequirement_IsFirstEverConversationWithNPC}}]
      priority {curve {rowName=Tandem.Speech.Priority.FirstEverInteraction}}
     }
     {
      message=Which member of my family do I need to apologize for now?
      contextRequirements [{requirement {class=FortControllerRequirement_CheckNPCMatchConversationCount, properties {nPCUniqueIdTag {tagName=AISpawnerData.Type.Tandem.FruitFire}}}}]
      priority {curve {rowName=None}}
     }
     {
      message="I just love your hair!"
      contextRequirements [{requirement {
         class=FortControllerRequirement_HasCID
         properties {allowedCIds [
           &Character_IronLilac_Sly
           &Character_BrightDisk
           &Character_VoidRedemption_Rebel
          ]}
        }}]
      priority {value=2, curve {rowName=Tandem.Speech.Priority.OutfitReaction}}
     }
     {
      message="Whatever you're up to, just try not to break anything."
      contextRequirements [{requirement {class=FortControllerRequirement_HasCID, properties {allowedCIds [/BRCosmetics/Athena/Items/Cosmetics/Characters/Character_QueenTruth_Rind]}}}]
      priority {value=2, curve {rowName=Tandem.Speech.Priority.OutfitReaction}}
     }
     {
      message=I'm glad you're spending time with your brother. I think.
      contextRequirements [{requirement {class=FortControllerRequirement_HasCID, properties {allowedCIds [/BRCosmetics/Athena/Items/Cosmetics/Characters/Character_FrostGalore]}}}]
      priority {value=2, curve {rowName=Tandem.Speech.Priority.OutfitReaction}}
     }
     {
      message="Please be careful, Homie."
      contextRequirements [{requirement {class=FortControllerRequirement_HasCID, properties {allowedCIds [/BRCosmetics/Athena/Items/Cosmetics/Characters/Character_SilentNovel_Vane]}}}]
      priority {value=2, curve {rowName=Tandem.Speech.Priority.OutfitReaction}}
     }
     {
      message="Tell Linda I say hi!"
      contextRequirements [{requirement {class=FortControllerRequirement_HasCID, properties {allowedCIds [/BRCosmetics/Athena/Items/Cosmetics/Characters/Character_CyclopsPrey]}}}]
      priority {value=2, curve {rowName=Tandem.Speech.Priority.OutfitReaction}}
     }
    ]
   }
  }
 }
 ConversationLinkNode_2 {
  id=#6
  class=ConversationLinkNode
  properties {remoteEntryTag {tagName=Conversation.Entry.Season38.FruitFire.P04}}
 }
 ConversationLinkNode_4 {
  id=#5
  class=ConversationLinkNode
  properties {remoteEntryTag {tagName=Conversation.Entry.Season38.LAF.NPC.Collect}}
  sub_nodes [FortConversationRequirement_ControllerRequirement_1 {
    class=FortConversationRequirement_ControllerRequirement
    properties {
     nodeName="Require: Has not started a L&&F Quest"
     requirement {class=FortControllerRequirement_HasOwnedTag, properties {bInverseCondition=true, ownedTag {tagName=Athena.Quests.S38.LAF.QuestActive}}}
    }
   }]
 }
 ConversationLinkNode_0 {id=#4, class=ConversationLinkNode, properties {remoteEntryTag {tagName=Conversation.Entry.BasicServices}}}
 ConversationLinkNode_1 {id=#3, class=ConversationLinkNode, properties {remoteEntryTag {tagName=Conversation.Entry.NPCLibrary.Quests}}}
]
